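# Animation Blueprint for the FNCS Shield (S34) back bling that shifts and rolls the pack bone to follow the wearer's spine.
# Each update reads pelvis and spine bone deltas from the owner's mesh and maps their yaw to clamped alphas and offsets.
# bIsInFrontEnd selects the variable-driven offset plus a fixed lift in the front end.
# Otherwise, two stacked fixed additive bone adjustments apply, weighted by spine rotation and curvature alphas.
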AnimBlueprint /BRCosmetics/Accessories/FORT_Backpacks/Backpack_M_MED_FNCSShield_S34/Meshes/M_MED_FNCSShield_S34_Pack_AnimBP
target_skeleton=&/M_MED_FNCSShield_S34_Pack_Skeleton
graphs [
 AnimGraph {nodes [
   n1 = "Transform (Modify) Bone - Bone: pack"(
    ComponentPose=Local To Component(LocalPose=Use cached pose 'input pose'())
    Alpha=Get SpineRotationAlpha() -> (SpineRotationAlpha:double)
    Translation="0,-1,0"
    Rotation="0,0,20"
   ) {
    rotationMode=BMM_Additive
    rotationSpace=BCS_BoneSpace
    translationMode=BMM_Additive
    translationSpace=BCS_BoneSpace
    alphaBoolBlend {blendInTime=0.3, blendOutTime=1}
    boneToModify {boneName=pack}
    rotation {roll=20}
    translation {y=-1}
   }
   n2 = "Transform (Modify) Bone - Bone: pack"(
    ComponentPose=@n1
    Alpha=Get SpineCurvatureCompensationAlpha() -> (SpineCurvatureCompensationAlpha:double)
    Translation="0,-5,0"
    Rotation="0,0,10"
   ) {
    rotationMode=BMM_Additive
    rotationSpace=BCS_BoneSpace
    translationMode=BMM_Additive
    translationSpace=BCS_BoneSpace
    alphaBoolBlend {blendInTime=0.3, blendOutTime=1}
    boneToModify {boneName=pack}
    rotation {roll=10}
    translation {y=-5}
   }
   n3 = "Transform (Modify) Bone - Bone: pack"(
    ComponentPose=Local To Component(LocalPose=Use cached pose 'input pose'())
    Translation=Get backpack_translation()
    Rotation=Get backpack_rotate()
   ) {
    rotationMode=BMM_Additive
    rotationSpace=BCS_BoneSpace
    translationMode=BMM_Additive
    translationSpace=BCS_BoneSpace
    alphaBoolBlend {blendOption=Linear}
    boneToModify {boneName=pack}
   }
   n4 = "Transform (Modify) Bone - Bone: pack"(ComponentPose=@n3, Translation="0,0,3", Rotation="0,0,0") {
    translationMode=BMM_Additive
    translationSpace=BCS_BoneSpace
    alphaBoolBlend {blendOption=Linear}
    boneToModify {boneName=pack}
    translation {z=3}
   }
   Output Pose(Result=Blend Poses by bool(BlendPose_0=Component To Local(ComponentPose=@n4), BlendPose_1=Component To Local(ComponentPose=@n2), bActiveValue=Get bIsInFrontEnd()) {blendType=Linear}) {layerGroup=None}
   Save cached pose 'input pose'(Pose=Input Pose()) {cacheName=input pose}
  ]}
 EventGraph {nodes [
   Event BlueprintUpdateAnimation()
   > n1 = Sequence()
   n2 = GetOwnerSkeletalMeshComponent()
   n3 = GetDeltaTransformFromRefPose(self=@n2, BoneName=pelvis, BaseName=spine_05)
   Set SpineRotationAlpha(
    execute=@n1.then_1
    SpineRotationAlpha:double=MapRangeClamped(
     Value=BreakTransform!(InTransform=@n3) -> (Location, Rotation_Roll:float, Rotation_Pitch:float, Rotation_Yaw:float, Scale).Rotation_Yaw
     InRangeA=0
     InRangeB=15
    )
   ) -> (Output_Get:double)
   n4 = GetDeltaTransformFromRefPose(self=@n2, BoneName=spine_03, BaseName=spine_01)
   > Set SpineCurvatureCompensationAlpha(SpineCurvatureCompensationAlpha:double=MapRangeClamped(Value=BreakTransform(InTransform=@n4).Rotation_Yaw, InRangeA=0, InRangeB=14)) -> (Output_Get:double)
   n5 = GetOwnerSkeletalMeshComponent()
   n6 = BreakRotator(InRot=BreakTransform!(InTransform=GetDeltaTransformFromRefPose(self=@n5, BoneName=spine_05, BaseName=pelvis)) -> (Location, Rotation:Rotator, Scale).Rotation)
   Set backpack_rotate(execute=@n1.then_0, backpack_rotate=MakeRotator(Roll=MapRangeClamped(Value=@n6.Yaw, InRangeA=0, InRangeB=-30, OutRangeB=20)))
   > Set backpack_translation(backpack_translation=MakeVector(
     Y=MapRangeClamped(Value=@n6.Yaw, InRangeA=0, InRangeB=-30, OutRangeB=-5)
     Z=MapRangeClamped(Value=@n6.Yaw, InRangeA=0, InRangeB=-30, OutRangeB=5)
    ))
   TryGetPawnOwner()
   GetAlphaFromReferenceTransformDelta(SkeletalMeshComponent=@n2, BoneName=spine_03, BaseName=spine_01, InputRange=(X=-7,Y=-20), TwistAxis="1,0,0")
  ]}
]
variables [
 backpack_rotate {type=Rotator}
 backpack_translation {type=Vector}
 SpineCurvatureCompensationAlpha {type=number}
 SpineRotationAlpha {type=number}
]
